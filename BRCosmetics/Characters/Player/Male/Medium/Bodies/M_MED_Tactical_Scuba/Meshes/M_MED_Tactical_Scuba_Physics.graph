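# Physics asset for the male medium Tactical Scuba body, simulating secondary motion on what the bone names indicate is a chest-mounted breathing tube.
# A spine_05 capsule anchors a chain of a box on the tube's chest bone, then capsules on its jiggle and hose bones.
# The chest joint rotates freely, the jiggle joint has narrow angular limits, and the hose is rotationally locked to the jiggle body.
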PhysicsAsset /BRCosmetics/Characters/Player/Male/Medium/Bodies/M_MED_Tactical_Scuba/Meshes/M_MED_Tactical_Scuba_Physics
bodies {
 dyn_tubeAimChest [{shapeType=Box, extentX=2.7662, extentY=3.4512, extentZ=3.5259, center=<-2.0426 -1.0255 -0.2671>, rotation=<rot 17.9113 -90 22.1861>}]
 dyn_tubeAimHose [{shapeType=Capsule, length=3.2832, radius=1.7334, center=<0.1568 0.4319 0.1687>, rotation=<rot 0 -115.6544 -85.9531>}]
 dyn_tubeAimJiggle [{shapeType=Capsule, length=2.9986, radius=3.9693, center=<3.1424 4.2468 -0.0173>, rotation=<rot 0.2042 -171.3045 -90.0702>}]
 spine_05 [{shapeType=Capsule, length=22.2991, radius=8.9718}]
}
constraints [
 {
  bone1Name=dyn_tubeAimChest
  bone2Name=spine_05
  swing1LimitDegrees=45
  swing1Motion=Free
  swing2LimitDegrees=45
  swing2Motion=Free
  twistLimitDegrees=45
  twistMotion=Free
 }
 {bone1Name=dyn_tubeAimJiggle, bone2Name=dyn_tubeAimChest, swing1LimitDegrees=16, swing2LimitDegrees=16, twistLimitDegrees=16}
 {
  bone1Name=dyn_tubeAimHose
  bone2Name=dyn_tubeAimJiggle
  swing1LimitDegrees=45
  swing1Motion=Locked
  swing2LimitDegrees=45
  swing2Motion=Locked
  twistLimitDegrees=45
  twistMotion=Locked
 }
]
